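# Prototype test SoundCue for the bug boss idle vocalization in the TidalCrane swarm boss content.
# It has no nodes or root, so it references no audio and is silent as it stands.
# It carries playback settings only: a dedicated attenuation asset, the boss-fight VOX sound class and a boosted volume.
# A concurrency override allows one instance at a time and blocks new ones while it plays.
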SoundCue /TidalCrane_Swarm_Boss/Assets/Sound/Proto/Bug_Boss_Idle_VOX_TEST
root=null
nodes []
properties {
 attenuationSettings=/TidalCrane_Swarm_Boss/Assets/Sound/Settings/Bug_Boss_Idle_VOX_Att
 bOverrideConcurrency=true
 maxDistance=11500
 soundClassObject=/Game/Sounds/FortSoundClassesAndModes/Fort_BossFight_SoundClass/Fort_BossFight_VOX
 volumeMultiplier=3
 concurrencyOverrides {maxCount=1, resolutionRule=PreventNew}
}
