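# Prototype test SoundCue for the bug boss's spawn movement in the Tidal Crane swarm boss fight.
# A single wave player plays one earth-emerge test wave, with no randomization or mixing.
# The cue routes through the boss fight sound class with boosted volume, using a shared boss movement attenuation asset and a long max audible distance.
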SoundCue /TidalCrane_Swarm_Boss/Assets/Sound/Proto/Bug_Boss_Spawn_Movement_TEST
root=WavePlayer_1
nodes [{id=WavePlayer_1, properties {soundWaveAssetPtr=&/Bug_Boss_Emerge_Earth_TEST_01}}]
properties {
 attenuationSettings=/TidalCrane_Swarm_Boss/Assets/Sound/Settings/Bug_Boss_Movement_Att
 maxDistance=21000
 soundClassObject=&Fort_BossFight
 volumeMultiplier=2
}
