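# Sound cue for a prototype bug boss spawn vocal in the Tidal Crane swarm boss content.
# A single wave player plays a test emerge scream, with no randomization or modulation.
# It uses the boss roar VOX attenuation for long-range audibility, routes to the boss fight VOX sound class, and boosts volume.
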SoundCue /TidalCrane_Swarm_Boss/Assets/Sound/Proto/Bug_Boss_Spawn_VOX_TEST
root=WavePlayer_0
nodes [{id=WavePlayer_0, properties {soundWaveAssetPtr=&/Bug_Boss_Emerge_Scream_TEST_01}}]
properties {
 attenuationSettings=/TidalCrane_Swarm_Boss/Assets/Sound/Settings/Bug_Boss_Roar_VOX_Att
 maxDistance=26000
 soundClassObject=/Game/Sounds/FortSoundClassesAndModes/Fort_BossFight_SoundClass/Fort_BossFight_VOX
 volumeMultiplier=2
}
